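# Niagara system for the Confetti vehicle boost, emitting streamers and confetti from the vehicle's boost port sockets.
# A ribbon emitter draws four ribbons per port, using age-driven sine offsets to wiggle them and fading brightness and width over lifetime.
# A sprite emitter spawns three groups of short-lived, multicolored pieces that rotate, fall and fade, with some brightened.
# Both inherit vehicle motion and support painted color variants; sprite velocity changes for frontend previews.
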NiagaraSystem /VehicleCosmetics/Boosts/Boost_Confetti/FX/NS_Boost_Conffeti
dependencies {
 usedDataInterfaces [
  &NiagaraDataInterfaceArrayColor
  &NiagaraDataInterfaceCurve
  &NiagaraDataInterfaceVectorCurve
  &NiagaraDataInterfaceSkeletalMesh
 ]
 usedDynamicInputs []
 usedModules [
  &self:SystemSpawnScript.SystemScriptSource.SystemScriptGraph.NiagaraNodeAssignment_0.SetVariables_#1
  &SystemState
 ]
}
emitters {
 RIBBONS001 {
  emitterUpdateScript [
   &EmitterState()
   &VC_FilteredSocketCount()
   &VC_RateQualityHandler()
   &VC_SpawnRibbon_v1(Number of Ribbons=$Emitter.PortCount, Spawn Time=-1, SpawnRate=60)
  ]
  particleSpawnScript [
   &VC_PaintedVariantsOverride(
    Black=<rgba 0.01 0.01 0.01 1>
    BurntSienna=<rgba 0.3229 0.0683 0 1>
    Crimson=<rgba 0.6823 0.0462 0.0634 1>
    DefaultColor=<rgba 0.4531 0.0094 0.0242 1>
    ForestGreen=<rgba 0 0.8177 0.0413 1>
    Grey=<rgba 0.4323 0.4323 0.4323 1>
    Lime=<rgba 0.5964 1 0.1927 1>
    Purple=<rgba 0.2244 0 0.8333 1>
    R1 Black:LinearColor=<rgba 0.01 0.01 0.01 1>
    R1 BurntSienna:LinearColor=<rgba 0.3229 0.0683 0 1>
    R1 Crimson:LinearColor=<rgba 0.6823 0.0462 0.0634 1>
    R1 DefaultColor=<rgba 0.1306 0.4152 0.5573 1>
    R1 ForestGreen:LinearColor=<rgba 0 0.8177 0.0413 1>
    R1 Lime:LinearColor=<rgba 0.5964 1 0.1927 1>
    R1 Purple:LinearColor=<rgba 0.2244 0 0.8333 1>
    R1 Saffron:LinearColor=<rgba 0.8021 0.6412 0.1128 1>
    R1 SkyBlue:LinearColor=<rgba 0.0195 0.5056 0.75 1>
    R2 Black:LinearColor=<rgba 0.01 0.01 0.01 1>
    R2 BurntSienna:LinearColor=<rgba 0.3229 0.0683 0 1>
    R2 Crimson:LinearColor=<rgba 0.6823 0.0462 0.0634 1>
    R2 DefaultColor:LinearColor=<rgba 0.1212 0.4323 0.1036 1>
    R2 ForestGreen:LinearColor=<rgba 0 0.8177 0.0413 1>
    R2 Lime:LinearColor=<rgba 0.5729 1 0.1458 1>
    R2 Purple:LinearColor=<rgba 0.2244 0 0.8333 1>
    R2 Saffron:LinearColor=<rgba 0.8021 0.6412 0.1128 1>
    R2 SkyBlue:LinearColor=<rgba 0.0195 0.5056 0.75 1>
    R3 Black:LinearColor=<rgba 0.01 0.01 0.01 1>
    R3 BurntSienna:LinearColor=<rgba 0.3229 0.0683 0 1>
    R3 Crimson:LinearColor=<rgba 0.6823 0.0462 0.0634 1>
    R3 DefaultColor:LinearColor=<rgba 0.4844 0.4844 0.1236 1>
    R3 ForestGreen:LinearColor=<rgba 0 0.8177 0.0413 1>
    R3 Grey:LinearColor=<rgba 0.3229 0.3229 0.3229 1>
    R3 Lime:LinearColor=<rgba 0.5729 1 0.1458 1>
    R3 Purple:LinearColor=<rgba 0.2244 0 0.8333 1>
    R3 Saffron:LinearColor=<rgba 1 0.7873 0.0885 1>
    R3 SkyBlue:LinearColor=<rgba 0.0195 0.5056 0.75 1>
    Saffron=<rgba 0.8021 0.6412 0.1128 1>
    SkyBlue=<rgba 0.0195 0.5056 0.75 1>
    Use R1 Painted Overrides=true
    Use R2 Painted Overrides=true
    Use R3 Painted Overrides=true
   )
   &VC_InitializeRibbon_v1(
    Color=$Particles.Color
    Enable Skeletal Position=true
    Lifetime=1.5
    Offset Amount=20
    Offset Axis:ENiagaraOrientationAxis=::XAxis
    Ribbon Width=10
   )
   &AddVelocity(Velocity=<-1500 0 50>) {module=AddVelocity001}
   /VehicleCosmetics/FX/VC_FXParents/Modules/DLR_ProductPreviews(GameModeID:NiagaraInt32=$System.GameModeID, bPreviewVehicleSpeed:NiagaraBool=$System.bPreviewVehicleSpeed)
  ]
  particleUpdateScript [
   &ParticleState()
   &InheritSourceMovement(
    Applied Position Scale=&Multiply_VectorByFloat(
     Vector=<1 1 1>
     Float=&FloatFromCurve(FloatCurve=DataInterface() {propertyValues {Curve {
         defaultValue=1
         keys [{interpMode=RCIM_Cubic, tangentMode=RCTM_User, time=0, value=1}, {interpMode=RCIM_Cubic, tangentMode=RCTM_User, time=1, value=0}]
        }}})
    )
    ApplyForce=false
    ApplyPosition=true
    ApplyPositionFalloff=false
   ) {module=InheritSourceMovement001}
   &ScaleRibbonWidth(Ribbon Width Scale=&FloatFromCurve(FloatCurve=DataInterface() {propertyValues {Curve {keys [<at 0.2 1 RCIM_Cubic>,<at 0.7 0 RCIM_Cubic>]}}}))
   &InheritSourceMovement(
    Applied Position Scale=&VectorFromFloat(Value=&FloatFromCurve(FloatCurve=DataInterface() {propertyValues {Curve {
         defaultValue=1
         keys [
          {arriveTangent=-1.7885, interpMode=RCIM_Cubic, leaveTangent=-1.7885, tangentMode=RCTM_User, time=0.05, value=1}
          {interpMode=RCIM_Cubic, tangentMode=RCTM_User, time=1, value=0}
         ]
        }}}))
    ApplyForce=false
    ApplyPosition=true
    ApplyPositionFalloff=false
   ) {enabled=false}
   &SolveForcesAndVelocity(Clamp Velocity=false, Limit Acceleration=false)
   /SetVariables(
    Particles.0-RibbonColor:LinearColor=/DealershipSharedLibrary/FX/Niagara/DynamicInputs/MultiplyLinearColorByVectorAndFloat(
     Color to Scale:LinearColor=$Particles.Color
     RGB Scale:Vector3f=&VectorFromFloat(Value=&FloatFromCurve(FloatCurve=DataInterface() {propertyValues {Curve {keys [<at 0 1>,<at 0.5 0.2>]}}}))
     Alpha Scale:NiagaraFloat=1
    )
    Particles.1-RibbonColor:LinearColor=&MultiplyLinearColorByVectorAndFloat(
     Color to Scale:LinearColor=$Particles.R1 Color
     RGB Scale:Vector3f=&VectorFromFloat(Value=&FloatFromCurve(FloatCurve=DataInterface() {propertyValues {Curve {keys [<at 0 1>,<at 0.5 0.2>]}}}))
     Alpha Scale:NiagaraFloat=1
    )
    Particles.2-RibbonColor:LinearColor=&MultiplyLinearColorByVectorAndFloat(
     Color to Scale:LinearColor=$Particles.R2 Color
     RGB Scale:Vector3f=&VectorFromFloat(Value=&FloatFromCurve(FloatCurve=DataInterface() {propertyValues {Curve {keys [<at 0 1>,<at 0.5 0.2>]}}}))
     Alpha Scale:NiagaraFloat=1
    )
    Particles.3-RibbonColor:LinearColor=&MultiplyLinearColorByVectorAndFloat(
     Color to Scale:LinearColor=$Particles.R3 Color
     RGB Scale:Vector3f=&VectorFromFloat(Value=&FloatFromCurve(FloatCurve=DataInterface() {propertyValues {Curve {keys [<at 0 1>,<at 0.5 0.2>]}}}))
     Alpha Scale:NiagaraFloat=1
    )
   ) {isSetParameters=true}
   /SetVariables(
    Particles.PosNoodle1:NiagaraPosition=&AddVectorToPosition(
     Position=$Particles.Position
     Vector=&TransformVector(
      Vector=&MakeVector(
       X=0
       Y=&Sine(
        Normalized Angle=&Add_Float(
         A=&Multiply_Float(A=$Particles.NormalizedAge, B=-1.3)
         B=&Multiply_Float(A=$Emitter.LoopedAge, B=1.8)
        )
        Period=0.5
        Scale=&FloatFromCurve(FloatCurve=DataInterface() {propertyValues {Curve {defaultValue=0.3, keys [<at 0 15>, {tangentMode=RCTM_User, time=0.6, value=35}, {tangentMode=RCTM_User, time=1, value=60}]}}})
        Bias=0.1
       )
       Z=&Sine(
        Normalized Angle=&Add_Float(
         A=&Multiply_Float(A=$Particles.NormalizedAge, B=-1.3)
         B=&Multiply_Float(A=$Emitter.LoopedAge, B=2)
        )
        Period=0.6
        Scale=&FloatFromCurve(FloatCurve=DataInterface() {propertyValues {Curve {defaultValue=0.3, keys [<at 0 15>, {tangentMode=RCTM_User, time=0.6, value=35}, {tangentMode=RCTM_User, time=1, value=60}]}}})
        Bias=0
       )
      )
      Destination Space=::Simulation
     )
    )
    Particles.PosNoodle2:NiagaraPosition=&AddVectorToPosition(
     Position=$Particles.Position
     Vector=&TransformVector(
      Vector=&MakeVector(
       X=0
       Y=&Sine(
        Normalized Angle=&Add_Float(
         A=&Multiply_Float(A=$Particles.NormalizedAge, B=-1.2)
         B=&Multiply_Float(A=$Emitter.LoopedAge, B=1.8)
        )
        Period=0.6
        Scale=&FloatFromCurve(FloatCurve=DataInterface() {propertyValues {Curve {keys [
             {interpMode=RCIM_Cubic, leaveTangentWeight=0.3333, tangentMode=RCTM_User, time=0, value=10}
             {arriveTangentWeight=9.3975, interpMode=RCIM_Cubic, tangentMode=RCTM_User, time=1, value=50}
            ]}}})
        Bias=0
       )
       Z=&Sine(
        Normalized Angle=&Add_Float(
         A=&Multiply_Float(A=$Particles.NormalizedAge, B=1)
         B=&Multiply_Float(A=$Emitter.LoopedAge, B=-1.3)
        )
        Period=0.6
        Scale=&FloatFromCurve(FloatCurve=DataInterface() {propertyValues {Curve {keys [<at 0 9.3837 RCIM_Cubic>,<at 1 40 RCIM_Cubic>]}}})
        Bias=0
       )
      )
      Destination Space=::Simulation
      Apply Scale=true
     )
    )
    Particles.PosNoodle3:NiagaraPosition=&AddVectorToPosition(
     Position=$Particles.Position
     Vector=&TransformVector(
      Vector=&MakeVector(
       X=0
       Y=&Sine(
        Normalized Angle=&Add_Float(
         A=&Multiply_Float(A=$Particles.NormalizedAge, B=1.5)
         B=&Multiply_Float(A=$Emitter.LoopedAge, B=-1.5)
        )
        Period=0.5
        Scale=&FloatFromCurve(FloatCurve=DataInterface() {propertyValues {Curve {defaultValue=0.3, keys [<at 0 10>, {tangentMode=RCTM_User, time=0.6, value=35}, {tangentMode=RCTM_User, time=1, value=60}]}}})
        Bias=0.1
       )
       Z=&Sine(
        Normalized Angle=&Add_Float(
         A=&Multiply_Float(A=$Particles.NormalizedAge, B=1.6)
         B=&Multiply_Float(A=$Emitter.LoopedAge, B=-1.8)
        )
        Period=0.6
        Scale=&FloatFromCurve(FloatCurve=DataInterface() {propertyValues {Curve {defaultValue=0.3, keys [<at 0 10>, {tangentMode=RCTM_User, time=0.6, value=35}, {tangentMode=RCTM_User, time=1, value=60}]}}})
        Bias=0
       )
      )
      Destination Space=::Simulation
      Apply Scale=true
     )
    )
    Particles.PosNoodle4:NiagaraPosition=&AddVectorToPosition(
     Position=$Particles.Position
     Vector=&TransformVector(
      Vector=&MakeVector(
       X=0
       Y=&Sine(
        Normalized Angle=&Add_Float(
         A=&Multiply_Float(A=$Particles.NormalizedAge, B=1.4)
         B=&Multiply_Float(A=$Emitter.LoopedAge, B=-1.7)
        )
        Period=0.7
        Scale=&FloatFromCurve(FloatCurve=DataInterface() {propertyValues {Curve {keys [
             {interpMode=RCIM_Cubic, leaveTangentWeight=0.3333, tangentMode=RCTM_User, time=0, value=15}
             {arriveTangentWeight=9.3975, interpMode=RCIM_Cubic, tangentMode=RCTM_User, time=1, value=50}
            ]}}})
        Bias=0
       )
       Z=&Sine(
        Normalized Angle=&Add_Float(
         A=&Multiply_Float(A=$Particles.NormalizedAge, B=-1.2)
         B=&Multiply_Float(A=$Emitter.LoopedAge, B=1.5)
        )
        Period=0.8
        Scale=&FloatFromCurve(FloatCurve=DataInterface() {propertyValues {Curve {keys [<at 0 15 RCIM_Cubic>,<at 1 40 RCIM_Cubic>]}}})
        Bias=0
       )
      )
      Destination Space=::Simulation
      Apply Scale=true
     )
    )
   ) {isSetParameters=true}
   &VC_UpdateRibbon_v1(Enable Skeletal Position=true)
  ]
  rendererClasses [&NiagaraRibbonRendererProperties]
  renderers [
   {rendererClass=&NiagaraRibbonRendererProperties}
   {rendererClass=&NiagaraRibbonRendererProperties, rendererIndex=1}
   {rendererClass=&NiagaraRibbonRendererProperties, rendererIndex=2}
   {rendererClass=&NiagaraRibbonRendererProperties, rendererIndex=3}
  ]
 }
 SPRITES {
  emitterUpdateScript [
   &EmitterState()
   &VC_FilteredSocketCount()
   &SpawnRate(SpawnRate=35, Use Spawn Probability=false)
   &SpawnRate(SpawnGroup=1, SpawnRate=65, Use Spawn Probability=false) {module=SpawnRate002}
   &SpawnRate(SpawnGroup=2, SpawnRate=65, Use Spawn Probability=false) {module=SpawnRate003}
  ]
  particleSpawnScript [
   /SetVariables(Particles.VisTag:NiagaraInt32=$Emitter.SpawnGroup) {isSetParameters=true}
   &VC_PaintedVariantsOverride(
    Black=<rgba 0.001 0.001 0.001 1>
    BurntSienna=<rgba 0.0781 0.0197 0.0041 1>
    Cobalt=<rgba 0 0.0555 3 1>
    Crimson=<rgba 0.5156 0 0.014 1>
    DefaultColor=&SelectLinearColorFromArray(
     Array Sampling Mode=::NewEnumerator0
     Direct Array Index=?
     Color Selection Array=DataInterface() {propertyValues {ColorData [
        <rgba 0.2552 0.0053 0.0136 1>
        <rgba 0.1912 0.1529 0 1>
        <rgba 0 0.1042 0.0451 1>
        <rgba 0.1927 0.1532 0 1>
        <rgba 0.6724 0.2346 0 1>
        <rgba 0 0.1046 0.0452 1>
        <rgba 0.2448 0.0854 0 1>
        <rgba 0.5208 0.5208 0.5208 1>
        <rgba 0.4508 0.3613 0 1>
        <rgba 0.007 0.1301 0 1>
        <rgba 0 0.1042 0.0451 1>
        <rgba 0.1912 0.1529 0 1>
        <rgba 0.2448 0.0854 0 1>
       ]}}
     Random Seed=0
    )
    Lime=<rgba 0.2874 0.5156 0.0591 1>
    Orange=<rgba 1 0.1478 0 1>
    Pink=<rgba 0.3594 0.0431 0.2443 1>
    Purple=<rgba 0.2266 0 0.5521 1>
    Saffron=<rgba 0.6771 0.5347 0.067 1>
    SkyBlue=<rgba 0 0.5164 0.776 1>
   )
   &InitializeParticle(
    Color Mode=::NewEnumerator0
    Lifetime=&RandomRangeFloat(Minimum=0.4, Maximum=0.6)
    Position Offset=<300 0 0>
    Sprite Rotation Mode=::NewEnumerator2
    Sprite Size Mode=::NewEnumerator1
    Uniform Sprite Size Max=12
    Uniform Sprite Size Min=18
    UsePositionOffset=false
    Write Material Random=true
   )
   &PartitionParticles(Number Of Partitions=$Emitter.PortCount, Particles Per Partition=3, Partition Mode:ENiagara_PartitionMode=::NewEnumerator1)
   &SkeletalMeshLocation(
    Bone / Socket Index=$Particles.PartitionParticles.Partition
    Bone / Socket Sampling Mode=::NewEnumerator5
    MaskSpawnGroup=false
    Mesh Sampling Type=::NewEnumerator6
    Skeletal Mesh=$System.SkeletalMesh
   )
   &AddVelocity(
    Velocity=&MakeCustomVectorFromBool(
     Bool=$User.bIsFrontendPreview
     True Vector=&RandomRangeVector(Minimum=<-1000 200 300>, Maximum=<-1500 -200 50>)
     False Vector=&RandomRangeVector(Minimum=<-600 200 300>, Maximum=<-800 -200 50>)
    )
    Velocity Speed Scale=2
   )
   &ShapeLocation(MaskSpawnGroup=false, Override Seed=false, Sphere Radius=50, UseOwnerScale=true)
   /SetVariables(Particles.2-SpriteSize:Vector2f=&RandomRangeVector2D(Minimum=<5 10>, Maximum=<10 20>)) {isSetParameters=true}
   &DLR_ProductPreviews(GameModeID:NiagaraInt32=$System.GameModeID, bPreviewVehicleSpeed:NiagaraBool=$System.bPreviewVehicleSpeed)
  ]
  particleUpdateScript [
   &ParticleState()
   &SpriteRotationRate(Rotation Rate=&RandomRangeFloat(Minimum=-50, Maximum=50))
   &InheritSourceMovement(
    Applied Position Scale=&Multiply_VectorByFloat(
     Vector=<1 1 1>
     Float=&FloatFromCurve(FloatCurve=DataInterface() {propertyValues {Curve {keys [<at 0.1 1 RCIM_Cubic>, {arriveTangent=-3.4326, interpMode=RCIM_Cubic, leaveTangent=-3.4326, tangentMode=RCTM_User, time=0.9, value=0}]}}})
    )
    ApplyForce=false
    ApplyPosition=true
    ApplyPositionFalloff=false
   )
   &AccelerationForce(Acceleration=&VectorFromCurve(
     VectorCurve=DataInterface() {propertyValues {XCurve {keys [<at 0 0>,<at 1 0>]}, YCurve {keys [<at 0 0>,<at 1 0>]}, ZCurve {keys [<at 0 0 RCIM_Cubic>,<at 1 -1 RCIM_Cubic>]}}}
     Scale Curve=<1 1 5000>
    ))
   &InheritSourceMovement(
    Applied Position Scale=&Multiply_VectorByFloat(
     Vector=<1 1 1>
     Float=&FloatFromCurve(FloatCurve=DataInterface() {propertyValues {Curve {
         defaultValue=1
         keys [
          {arriveTangent=-1.7885, interpMode=RCIM_Cubic, leaveTangent=-1.7885, tangentMode=RCTM_User, time=0.05, value=1}
          {interpMode=RCIM_Cubic, tangentMode=RCTM_User, time=1, value=0}
         ]
        }}})
    )
    ApplyForce=false
    ApplyPosition=true
    ApplyPositionFalloff=false
   ) {enabled=false, module=InheritSourceMovement001}
   /SetVariables(Particles.0-SpriteColor:LinearColor=&MultiplyLinearColorByVectorAndFloat(
     Color to Scale:LinearColor=$Particles.Color
     RGB Scale:Vector3f=&VectorFromFloat(Value=&MaskFloat(
       Mask Source=1
       Mask Parameter=&ModuloInt(A=&ReturnExecIndex(), B=8)
       Mask=4
       Masked Float=10
       Passthrough Float=&RandomRangeFloat(Minimum=3, Maximum=1)
      ))
     Alpha Scale:NiagaraFloat=&FloatFromCurve(FloatCurve=DataInterface() {propertyValues {Curve {keys [<at 0.5 1>,<at 1 0>]}}})
    )) {isSetParameters=true}
   &SolveForcesAndVelocity(Clamp Velocity=false, Limit Acceleration=false)
  ]
  rendererClasses [&NiagaraSpriteRendererProperties]
  renderers [
   {rendererClass=&NiagaraSpriteRendererProperties}
   {rendererClass=&NiagaraSpriteRendererProperties, rendererIndex=1}
   {rendererClass=&NiagaraSpriteRendererProperties, rendererIndex=2}
  ]
 }
}
summary {
 systemName=NS_Boost_Conffeti
 emitters [
  {emitterName=SPRITES, rendererClasses [&NiagaraSpriteRendererProperties]}
  {emitterName=RIBBONS001, rendererClasses [&NiagaraRibbonRendererProperties]}
 ]
 userVariables [
  User.bIsFrontendPreview {type=NiagaraBool}
  User.bIsPainted {type=NiagaraBool}
  User.Brightness {type=NiagaraFloat, defaultValue=1, description=""}
  User.Brightness2 {type=NiagaraFloat, defaultValue=1, description=""}
  User.ColorVariantID {type=NiagaraInt32}
  User.Desaturation {type=NiagaraFloat, defaultValue=0, description=""}
  User.PaintedColor {type=LinearColor, description="", defaultValue {struct=LinearColor, value=<rgba 1 1 1 1>}}
  User.PaintedScalar {type=NiagaraFloat, defaultValue=1, description=""}
  User.PortCount {type=NiagaraInt32, defaultValue=1}
 ]
}
