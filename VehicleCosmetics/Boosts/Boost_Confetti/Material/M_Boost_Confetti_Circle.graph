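# Translucent unlit Niagara sprite material that draws circular confetti particles for the Confetti vehicle boost.
# A dynamic parameter blends particle color toward an offset, optionally desaturated rainbow texture sample for the painted look and sets brightness.
# A sine of particle alpha makes brightness flicker over the alpha range before the shared emissive adjustment function.
# The circle mask is procedural, built from UV dot-product falloff and multiplied by particle alpha.
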Material /VehicleCosmetics/Boosts/Boost_Confetti/Material/M_Boost_Confetti_Circle
graph {
 properties {bUsedWithNiagaraSprites=true, blendMode=BLEND_Translucent, pixelDepthOffsetMode=PDOM_Legacy, shadingModel=MSM_Unlit, usedShadingModels=Unlit}
 nodes [
  Constant()
  Constant()
  Constant()
  MaterialExpressionVertexColor_0 = ParticleColor()
  MaterialExpressionConstantClamp_3 = Saturate(None=@MaterialExpressionVertexColor_0.RGB)
  n1 = DynamicParameter() {paramNames [Color, Color_Brightness, Desaturate, Painted]}
  n2 = Add(A=TextureCoordinate() {vTiling=0}, B=@n1.Color)
  n3 = LinearInterpolate(
   A=@MaterialExpressionConstantClamp_3
   B=Desaturation(None=TextureSample(UVs=@n2) {texture=/DealershipSharedLibrary/FX/Textures/Diffuse/T_VCVFX_Rainbow_D}.RGB, Fraction=@n1.Desaturate)
   Alpha=@n1.Painted
  )
  n4 = ConstantBiasScale(None=Sine(None=@MaterialExpressionVertexColor_0.A) {period=0.25}) {bias=0.75}
  MaterialExpressionConstantClamp_2 = Saturate(None=ConstantBiasScale(None=@n4) {bias=-0.5, scale=1})
  n5 = Add(A=Add(A=@n4) {constB=0.5}, B=Multiply(A=@MaterialExpressionConstantClamp_2) {constB=4})
  MaterialFunctionCall(Emissive=Multiply(A=Multiply(A=@n3, B=@n5), B=@n1.Color_Brightness)) {materialFunction=&MF_FX_EmissiveAdjustments}
  n6 = Add(A=TextureCoordinate() {uTiling=0.5, vTiling=0.5}) {constB=-0.25}
  n7 = OneMinus(None=DotProduct(A=@n6, B=@n6))
  MaterialExpressionConstantClamp_1 = Saturate(None=ConstantBiasScale(None=Power(Base=@n7) {constExponent=20}) {bias=-0.5, scale=2})
  MaterialExpressionConstantClamp_0 = Saturate(None=Multiply(A=@MaterialExpressionConstantClamp_1) {constB=8})
  Multiply(A=@MaterialExpressionConstantClamp_0, B=@MaterialExpressionVertexColor_0.A)
 ]
}
