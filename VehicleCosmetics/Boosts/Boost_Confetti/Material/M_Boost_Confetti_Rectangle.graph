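# Unlit translucent Niagara sprite material for rectangular boost confetti.
# Dynamic parameters blend particle color with an offset, desaturatable rainbow texture and set brightness before MF_FX_EmissiveAdjustments.
# A sine of particle alpha drives both glints and mask width, suggesting tumbling, while pixel depth softens the procedural rectangle's edges with distance.
# The graph holds a second, near-duplicate copy of the color and mask chains.
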Material /VehicleCosmetics/Boosts/Boost_Confetti/Material/M_Boost_Confetti_Rectangle
graph {
 properties {bUsedWithNiagaraSprites=true, blendMode=BLEND_Translucent, pixelDepthOffsetMode=PDOM_Legacy, shadingModel=MSM_Unlit, usedShadingModels=Unlit}
 nodes [
  n1 = ParticleColor()
  n2 = DynamicParameter() {paramNames [Color, Color_Brightness, Desaturate, ColorShift]}
  n3 = Add(A=TextureCoordinate() {vTiling=0}, B=@n2.Color)
  n4 = LinearInterpolate(
   A=Saturate(None=@n1.RGB)
   B=Desaturation(None=TextureSample(UVs=@n3) {texture=/DealershipSharedLibrary/FX/Textures/Diffuse/T_VCVFX_Rainbow_D}.RGB, Fraction=@n2.Desaturate)
   Alpha=@n2.ColorShift
  )
  n5 = ConstantBiasScale(None=Sine(None=@n1.A) {period=0.25}) {bias=0.75}
  n6 = ConstantBiasScale(None=@n5) {bias=-0.5, scale=1}
  n7 = Add(A=Add(A=@n5) {constB=0.5}, B=Multiply(A=Saturate(None=@n6)) {constB=16})
  MaterialFunctionCall(Emissive=Multiply(A=@n2.Color_Brightness, B=Multiply(A=@n4, B=@n7))) {materialFunction=&MF_FX_EmissiveAdjustments}
  Multiply()
  n8 = PixelDepth()
  n9 = Constant() {r=0.0002}
  n10 = OneMinus(None=Saturate(None=Multiply(A=@n8.R, B=@n9)))
  n11 = Add(A=TextureCoordinate() {uTiling=2, vTiling=2}) {constB=-1}
  n12 = Add(A=Abs(None=ComponentMask(None=@n11) {r=true}), B=@n5)
  n13 = ComponentMask(None=@n11) {g=true}
  n14 = AppendVector(A=Saturate(None=OneMinus(None=@n12)), B=OneMinus(None=Abs(None=@n13)))
  n15 = Saturate(None=Multiply(A=ConstantBiasScale(None=@n10) {bias=0.25, scale=8}, B=Abs(None=@n14)))
  Multiply(A=ComponentMask(None=@n15) {r=true}, B=ComponentMask(None=@n15) {g=true})
  n16 = ParticleColor()
  n17 = TextureCoordinate() {vTiling=0}
  n18 = DynamicParameter() {paramNames [Color, Color_Brightness, Desaturate, ColorShift]}
  n19 = Desaturation(
   None=TextureSample(UVs=Add(A=@n17, B=@n18.Color)) {texture=&T_VCVFX_Rainbow_D}.RGB
   Fraction=@n18.Desaturate
  )
  n20 = ConstantBiasScale(None=Sine(None=@n16.A) {period=0.25}) {bias=0.75}
  n21 = Saturate(None=ConstantBiasScale(None=@n20) {bias=-0.5, scale=1})
  Multiply(
   A=LinearInterpolate(A=@n16.RGB, B=Multiply(A=@n19, B=@n18.Color_Brightness), Alpha=@n18.ColorShift)
   B=Add(A=Add(A=@n20) {constB=0.5}, B=Multiply(A=@n21) {constB=16})
  )
  n22 = PixelDepth()
  n23 = Constant() {r=0.0002}
  n24 = OneMinus(None=Saturate(None=Multiply(A=@n22.R, B=@n23)))
  n25 = Add(A=TextureCoordinate() {uTiling=2, vTiling=2}) {constB=-1}
  n26 = Add(A=Abs(None=ComponentMask(None=@n25) {r=true}), B=@n20)
  n27 = ComponentMask(None=@n25) {g=true}
  n28 = AppendVector(A=Saturate(None=OneMinus(None=@n26)), B=OneMinus(None=Abs(None=@n27)))
  n29 = Saturate(None=Multiply(A=ConstantBiasScale(None=@n24) {bias=0.25, scale=8}, B=Abs(None=@n28)))
  Multiply(A=ComponentMask(None=@n29) {r=true}, B=ComponentMask(None=@n29) {g=true})
 ]
}
