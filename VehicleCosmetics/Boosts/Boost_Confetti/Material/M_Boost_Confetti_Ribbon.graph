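# Unlit masked material for the Niagara ribbons of the Confetti vehicle boost.
# Procedural UV edge and tip masks, plus a fast-tiled panning mask, shape particle color fed into a shared emissive-adjustment function.
# Summed sine and cosine waves build a transformed offset vector, with frequency, speed and amplitude from dynamic parameters and scaling along the ribbon.
# A rainbow lookup with offset, desaturation and brightness parameters also tints particle color.
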Material /VehicleCosmetics/Boosts/Boost_Confetti/Material/M_Boost_Confetti_Ribbon
graph {
 properties {
  bUsedWithNiagaraRibbons=true
  blendMode=BLEND_Masked
  pixelDepthOffsetMode=PDOM_Legacy
  refractionMethod=RM_2DOffset
  shadingModel=MSM_Unlit
  usedShadingModels=Unlit
 }
 nodes [
  n1 = ParticleColor()
  n2 = TextureCoordinate()
  n3 = ComponentMask(None=@n2) {g=true}
  n4 = OneMinus(None=@n3)
  MaterialExpressionConstantClamp_1 = Saturate(None=Multiply(A=Multiply(A=@n3, B=@n4), B=Constant() {r=8}))
  n5 = OneMinus(None=@MaterialExpressionConstantClamp_1)
  n6 = ComponentMask(None=@n2) {r=true}
  n7 = OneMinus(None=@n6)
  MaterialExpressionConstantClamp_2 = Saturate(None=Multiply(A=@n6, B=Constant() {r=155}))
  n8 = Power(Base=Multiply(A=@MaterialExpressionConstantClamp_1, B=Multiply(A=@n7, B=@MaterialExpressionConstantClamp_2)), Exp=Constant() {r=2})
  n9 = Add(A=Multiply(A=Constant() {r=2}, B=Multiply(A=@n5, B=@n8)), B=@n8)
  n10 = TextureCoordinate() {uTiling=1024, vTiling=0}
  n11 = Multiply(
   A=@n9
   B=TextureSample(UVs=Panner(Coordinate=@n10) {speedX=-0.1}) {texture=/DealershipSharedLibrary/FX/Textures/Gradient_Masks/T_VCVFX_Time_Mask_A}.R
  )
  MaterialExpressionConstantClamp_0 = Saturate(None=Multiply(A=@n1.RGB, B=Add(A=@n9, B=@n11)))
  MaterialFunctionCall(Emissive=@MaterialExpressionConstantClamp_0) {materialFunction=&MF_FX_EmissiveAdjustments}
  n12 = DynamicParameter() {parameterIndex=1}
  n13 = Multiply(A=@n12.G, B=OneMinus(None=Time()))
  n14 = ComponentMask(None=Multiply(A=TextureCoordinate() {coordinateIndex=1, uTiling=3}, B=@n12.R)) {r=true}
  n15 = Add(A=@n13, B=@n14)
  n16 = Add(A=Cosine(None=@n15), B=Cosine(None=@n15) {period=2})
  n17 = Add(A=@n14, B=@n13)
  n18 = Sine(None=@n17)
  n19 = Sine(None=@n17) {period=2}
  n20 = AppendVector(A=AppendVector(A=@n16, B=Add(A=@n18, B=@n19)), B=@n16)
  n21 = TextureCoordinate() {uTiling=6}
  MaterialExpressionConstantClamp_3 = Saturate(None=Add(A=ComponentMask(None=@n21) {r=true}, B=Constant() {r=0.35}))
  Multiply(A=Multiply(A=Transform(None=@n20), B=@n12.B), B=@MaterialExpressionConstantClamp_3)
  n22 = DynamicParameter() {paramNames [color, Color_Brightness, Desaturate, Tile]}
  n23 = TextureSample(UVs=Add(A=TextureCoordinate() {vTiling=0}, B=@n22.color)) {texture=/DealershipSharedLibrary/FX/Textures/Diffuse/T_VCVFX_Rainbow_D}
  Multiply(A=@n1.RGB, B=Multiply(A=Desaturation(None=@n23.RGB, Fraction=@n22.Desaturate), B=@n22.Color_Brightness))
  TextureSample(UVs=Panner(Coordinate=TextureCoordinate() {uTiling=64, vTiling=0}) {speedX=-0.1}) {samplerType=SAMPLERTYPE_LinearColor, texture=/VehicleCosmetics/Boosts/Boost_Confetti/Texture/T_VCVFX_Boost_Confetti_NewGraph_N}
  MaterialExpressionVertexColor_0 = ParticleColor()
 ]
}
