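# Unlit translucent Niagara sprite material for triangle confetti particles in the Confetti vehicle boost cosmetic.
# Dynamic parameters offset a V-flattened lookup into a rainbow texture, desaturate it, blend it with saturated particle color via Param4, and scale brightness.
# The result feeds a shared emissive-adjustment function.
# A triangle texture, boosted and clamped into a hard-edged mask, is multiplied by particle alpha for opacity.
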Material /VehicleCosmetics/Boosts/Boost_Confetti/Material/M_Boost_Confetti_Triangle
graph {
 properties {bUsedWithNiagaraSprites=true, blendMode=BLEND_Translucent, pixelDepthOffsetMode=PDOM_Legacy, shadingModel=MSM_Unlit, usedShadingModels=Unlit}
 nodes [
  Constant()
  MaterialExpressionVertexColor_1 = ParticleColor()
  n1 = Saturate(None=@MaterialExpressionVertexColor_1.RGB)
  n2 = TextureCoordinate() {vTiling=0}
  n3 = DynamicParameter() {paramNames [Color, Color_Brightness, Desaturated, Param4]}
  n4 = Desaturation(
   None=TextureSample(UVs=Add(A=@n2, B=@n3.Color)) {texture=/DealershipSharedLibrary/FX/Textures/Diffuse/T_VCVFX_Rainbow_D}.RGB
   Fraction=@n3.Desaturated
  )
  MaterialFunctionCall(Emissive=Multiply(A=LinearInterpolate(A=@n1, B=@n4, Alpha=@n3.Param4), B=@n3.Color_Brightness)) {materialFunction=&MF_FX_EmissiveAdjustments}
  MaterialExpressionConstantClamp_4 = Saturate(None=Multiply(A=TextureSample() {texture=/VehicleCosmetics/Boosts/Boost_Confetti/Texture/T_VCVFX_Boost_Confetti_Triangle}.RGB) {constB=16})
  Multiply(A=@MaterialExpressionVertexColor_1.A, B=@MaterialExpressionConstantClamp_4)
 ]
}
